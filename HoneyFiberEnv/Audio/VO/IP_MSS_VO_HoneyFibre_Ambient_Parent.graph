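# Mono MetaSound source, named as a parent, for looping ambient HoneyFibre NPC voice-over.
# A random delay between the TimeBetween bounds precedes a random pick from the input wave array.
# Each finished line feeds back through a deferred trigger variable to schedule the next, while a Bool output reports VO in progress.
# It sends to a music sidechain bus and limits concurrency to one per owner.
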MetaSoundSource /HoneyFiberEnv/Audio/VO/IP_MSS_VO_HoneyFibre_Ambient_Parent
inputs [
 In Array {type=WaveAsset:Array}
 TimeBetween.Max {type=Time}
 TimeBetween.Min {type=Time}
 UE.Source.OnPlay {type=Trigger}
]
interfaces [UE.OutputFormat.Mono:1.0, UE.Source:1.1]
outputs [
 bVOInProgress {type=Bool}
 "UE.OutputFormat.Mono.Audio:0" {type=Audio}
]
pages [Default {
  input_defaults {In Array=null, TimeBetween.Max=15, TimeBetween.Min=10, UE.Source.OnPlay=null}
  nodes [
   n1 = VariableDeferredAccessor.Trigger(Variable=InitVariable.Trigger().Variable)
   TriggerAnyTriggerAny2_0 = "TriggerAny.Trigger Any (2)"(In 0=^UE.Source.OnPlay, In 1=@n1.Value)
   n2 = UE.RandomTime(Max=^TimeBetween.Max, Min=^TimeBetween.Min, Next=@TriggerAnyTriggerAny2_0.Out)
   TriggerDelay_0 = UE.Trigger Delay(In=@n2.On Next, Delay Time=@n2.Value)
   ArrayRandomGet_0 = Array.Random Get:WaveAsset:Array(In Array=^In Array, Next=@TriggerDelay_0.Out)
   WavePlayer_0 = UE.Wave Player:Mono(Play=@ArrayRandomGet_0.On Next, Wave Asset=@ArrayRandomGet_0.Value) {also {version [1,1]}}
   ValueValue_0 = Value.Value:Bool(Reset=@WavePlayer_0.On Finished, Set=@WavePlayer_0.On Play)
   VariableMutator.Trigger(Variable=@n1.Variable, Value=@WavePlayer_0.On Finished)
  ]
  links [{dst="@bVOInProgress", src=ValueValue_0.Output Value}, {dst="@UE.OutputFormat.Mono.Audio:0", src=WavePlayer_0.Out Mono}]
 }]
properties {
 assetImportData=&self:AssetImportData
 bDebug=true
 bLooping=true
 bOverrideConcurrency=true
 duration=10000
 sampleRate=48000
 soundClassObject=&Fort_VO_NPC
 busSends [{audioBus=&AB_Sidechain_FortMusic_Vol, sendLevel=1}]
 concurrencyOverrides {bLimitToOwner=true, maxCount=1}
}
