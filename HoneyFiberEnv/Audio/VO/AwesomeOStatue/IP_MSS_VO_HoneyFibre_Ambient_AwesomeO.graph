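# Mono MetaSound Source preset that provides ambient voice lines for the AwesomeO statue in the HoneyFiber environment.
# It inherits the shared HoneyFibre ambient VO parent graph, supplying four statue wave assets and min/max times between lines.
# The parent exposes a VO-in-progress output alongside the audio.
# The looping source uses NPC VO sound class, AwesomeO-specific attenuation and a concurrency override limiting it to one instance per owner.
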MetaSoundSource /HoneyFiberEnv/Audio/VO/AwesomeOStatue/IP_MSS_VO_HoneyFibre_Ambient_AwesomeO
preset_parent=/HoneyFiberEnv/Audio/VO/IP_MSS_VO_HoneyFibre_Ambient_Parent
inputs [
 In Array {type=WaveAsset:Array}
 TimeBetween.Max {type=Time}
 TimeBetween.Min {type=Time}
 UE.Source.OnPlay {type=Trigger}
]
interfaces [UE.OutputFormat.Mono:1.0, UE.Source:1.1]
outputs [
 bVOInProgress {type=Bool}
 "UE.OutputFormat.Mono.Audio:0" {type=Audio}
]
pages [Default {input_defaults {
   TimeBetween.Max=15
   TimeBetween.Min=10
   UE.Source.OnPlay=null
   In Array [
    /HoneyFiberEnv/Audio/VO/AwesomeOStatue/Waves/IP_VO_HoneyFiber_AwesomeOStatue_Ambient_Greetings
    /HoneyFiberEnv/Audio/VO/AwesomeOStatue/Waves/IP_VO_HoneyFiber_AwesomeOStatue_Ambient_IAmAwesomeO
    /HoneyFiberEnv/Audio/VO/AwesomeOStatue/Waves/IP_VO_HoneyFiber_AwesomeOStatue_Ambient_OilWaste
    /HoneyFiberEnv/Audio/VO/AwesomeOStatue/Waves/IP_VO_HoneyFiber_AwesomeOStatue_Ambient_Understands
   ]
  }}]
properties {
 assetImportData=&self:AssetImportData
 attenuationSettings=&/IP_ATT_HoneyFibre_AwesomeO
 bDebug=true
 bLooping=true
 bOverrideConcurrency=true
 duration=10000
 maxDistance=5530
 sampleRate=48000
 soundClassObject=&Fort_VO_NPC
 concurrencyOverrides {bLimitToOwner=true, maxCount=1}
}
